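# Material function in the PrismParticle pickaxe folder that projects a panning texture onto surfaces from two side planes.
# Position (world position by default) is divided by TextureSize, masked to YZ and XZ, and offset by the Pan input.
# Both shared-sampler samples are blended by the absolute Y of the Normal input, sharpened with CheapContrast.
# It omits the XY projection, saving a sample over triplanar mapping, and outputs RGB.
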MaterialFunction /Game/Weapons/FORT_Melee/Pickaxe_PrismParticle/Materials/MF_Biplanar_Pan
graph {nodes [
  n1 = FunctionInput(Preview=Constant() {r=32}) {inputName=TextureSize, inputType=FunctionInput_Scalar, sortPriority=3, previewValue=<1 0 0 1>}
  n2 = Divide(
   A=FunctionInput(Preview=WorldPosition().XYZ) {bUsePreviewValueAsDefault=true, inputName=Position, sortPriority=1, previewValue=<0 0 1 1>}
   B=@n1
  )
  n3 = FunctionInput(Preview=Multiply(A=Time(), B=Constant2Vector())) {bUsePreviewValueAsDefault=true, inputName=Pan, inputType=FunctionInput_Vector2, sortPriority=4}
  n4 = Add(A=ComponentMask(None=@n2) {b=true, g=true}, B=@n3)
  n5 = FunctionInput(Preview=TextureObject() {samplerType=SAMPLERTYPE_Masks, texture=/Game/Characters/Player/Male/Large/Bodies/M_LRG_Slurp_Monster/Textures/T_SlurpBubble}) {inputName=Texture, inputType=FunctionInput_Texture2D}
  n6 = Add(A=ComponentMask(None=@n2) {b=true, r=true}, B=@n3)
  n7 = FunctionInput(Preview=VertexNormalWS()) {inputName=Normal, sortPriority=2}
  n8 = MaterialFunctionCall(
   In=Abs(None=ComponentMask(None=@n7) {g=true})
   Contrast=FunctionInput(Preview=Constant() {r=3}) {inputName=BlendContrast, inputType=FunctionInput_Scalar, sortPriority=5}
  ) {materialFunction=&CheapContrast}
  FunctionOutput(None=LinearInterpolate(
    A=TextureSample(UVs=@n4, Tex=@n5) {samplerSource=SSM_Wrap_WorldGroupSettings, samplerType=SAMPLERTYPE_Masks}.RGB
    B=TextureSample(UVs=@n6, Tex=@n5) {samplerSource=SSM_Wrap_WorldGroupSettings, samplerType=SAMPLERTYPE_Masks}.RGB
    Alpha=Saturate(None=@n8.Result)
   ))
 ]}
