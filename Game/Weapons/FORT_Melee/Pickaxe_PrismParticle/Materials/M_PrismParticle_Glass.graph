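# Translucent, refractive glass material for the PrismParticle pickaxe, with cracks that fade in through CrackProgress.
# Biplanar crack textures, projected in pre-skinned space so they follow the mesh, feed masks that alter color, emissive, opacity, normal and refraction.
# Emissive fakes reflections from a view-space sphere map, a desaturated tinted HDR cubemap and a sun/moon specular highlight.
# Fresnel on the detail normal drives base color and edge opacity, before the shared Fortnite base and character-opacity functions.
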Material /Game/Weapons/FORT_Melee/Pickaxe_PrismParticle/Materials/M_PrismParticle_Glass
graph {
 properties {
  bEnableResponsiveAA=true
  bScreenSpaceReflections=true
  bUseMaterialAttributes=true
  bUsedWithMeshDeformer=true
  bUsedWithNanite=true
  bUsedWithNiagaraMeshParticles=true
  bUsedWithSkeletalMesh=true
  blendMode=BLEND_Translucent
  physMaterial=/Game/FortressPhysicalMaterials/Glass
  pixelDepthOffsetMode=PDOM_Legacy
  refractionMethod=RM_IndexOfRefraction
  translucencyDirectionalLightingIntensity=5
  translucencyLightingMode=TLM_VolumetricPerVertexDirectional
  usedShadingModels=Default Lit
 }
 nodes [
  n1 = TextureObject() {samplerType=SAMPLERTYPE_Normal, texture=/Game/Weapons/FORT_Melee/Pickaxe_PrismParticle/Textures/T_Pickaxe_PrismParticle_Crack_N}
  n2 = ScalarParameter() {defaultValue=20, parameterName=CrackTextureSize}
  n3 = VertexInterpolator(VS=PreSkinnedNormal())
  n4 = ScalarParameter() {defaultValue=2, parameterName=CrackTextureContrast}
  n5 = VertexInterpolator(VS=PreSkinnedPosition())
  MaterialFunctionCall(TextureObject=@n1, TextureSize=@n2, Normal=@n3.PS, ProjectionTransitionContrast=@n4, WorldPosition=@n5.PS) {materialFunction=&WorldAlignedNormal}
  n6 = SetMaterialAttributes!(
   MaterialAttributes
   Base Color=NamedRerouteUsage()
   Metallic=ScalarParameter() {defaultValue=0.1, parameterName=Metallic}
   Specular=ScalarParameter() {defaultValue=1, parameterName=Specular}
   Roughness
   Emissive Color=NamedRerouteUsage()
   Opacity=NamedRerouteUsage()
   Opacity Mask
   Normal=NamedRerouteUsage()
   World Position Offset
   Refraction=LinearInterpolate(
    A=ScalarParameter() {defaultValue=1, parameterName=Refraction, sliderMax=2}
    B=ScalarParameter() {defaultValue=1.008, parameterName=CrackRefraction}
    Alpha=NamedRerouteUsage()
   ) {constB=1.1}
  ) {attributeSetTypes [
    #1
    #2
    #3
    #4
    #5
    #6
    #7
    #8
    #9
    #10
   ]}
  MaterialFunctionCall!(Material=MaterialFunctionCall!(Material attributes=@n6, IsGameplayAsset=StaticBool()) {materialFunction=/Game/Packages/Fortress_SharedMaterials/Base_Material_Functions/Fortnite_BaseFunction}.Material Attributes) {materialFunction=/Game/Global/Materials/Character_Parent_2023/Functions/BaseFN_CharacterOpacity_2023}
  n7 = TextureObject() {samplerType=SAMPLERTYPE_Masks, texture=/Game/Weapons/FORT_Melee/Pickaxe_PrismParticle/Textures/T_Pickaxe_PrismParticle_Crack_TA}
  MaterialFunctionCall(TextureObject=@n7, TextureSize=@n2, WorldPosition=@n5.PS, World Space Normal=@n3.PS, ProjectionTransitionContrast=@n4) {materialFunction=&WorldAlignedTexture}
  n8 = MaterialFunctionCall!(Texture=@n7, Position=@n5.PS, Normal=@n3.PS, TextureSize=@n2, Pan, BlendContrast=@n4) {materialFunction=&/MF_Biplanar_Pan}
  n9 = ComponentMask(None=@n8.Result) {g=true}
  n10 = ScalarParameter() {parameterName=CrackProgress}
  NamedRerouteDeclaration(None=LinearInterpolate(B=ConstantBiasScale(None=@n9) {bias=-0.5, scale=2}, Alpha=@n10)) {name=CrackRefraction, nodeColor=<rgba 0 0.5312 1 1>}
  n11 = ReflectionVectorWS()
  n12 = ConstantBiasScale(None=ComponentMask(None=Transform(None=@n11) {transformSourceType=TRANSFORMSOURCE_World, transformType=TRANSFORM_View}) {g=true, r=true})
  n13 = Constant2Vector() {g=-1, r=1}
  n14 = Multiply(
   A=TextureSampleParameter2D(UVs=Multiply(A=@n12, B=@n13)) {group=SphereReflection, parameterName=SphereReflectionMap}.RGB
   B=ScalarParameter() {defaultValue=1, group=SphereReflection, parameterName=Reflection Multiplier}
  )
  n15 = TextureSampleParameterCube!(UVs=ReflectionVectorWS() {bNormalizeCustomWorldNormal=true}, Apply View MipBias) {
   group=Cubemap
   parameterName=MaterialExpressionTextureSampleParameterCube_0
   texture=/Game/Packages/Fortress_Sky/TexturesHDR/T_hdrsky_test03_low
  }
  n16 = Multiply(A=@n15.RGB, B=VectorParameter() {group=Cubemap, parameterName=Cube Map Multiplier, defaultValue=<rgba 1.5 1.5 1.5 0>}.RGB) {constB=1.5}
  n17 = ComponentMask(None=MaterialFunctionCall() {materialFunction=/Engine/Functions/GetPostProcessFunctions/GetAmbientCubemapTint}.AmbientCubemapTint) {b=true, g=true, r=true}
  n18 = CollectionParameter() {
   collection=&FortniteMaterialParameters
   parameterName=SunAndMoonModelDirectionalVector
  }
  n19 = Clamp(None=DotProduct(A=ComponentMask(None=@n18) {b=true, g=true, r=true}, B=ReflectionVectorWS()))
  n20 = ScalarParameter() {defaultValue=70, group=FakeSpec, parameterName=Spec Highlight Power}
  n21 = Add(
   A=Desaturation(None=Multiply(A=@n16, B=@n17), Fraction=Constant() {r=0.5})
   B=Multiply(A=Power(Base=@n19, Exp=@n20) {constExponent=20}, B=ScalarParameter() {defaultValue=3, group=FakeSpec, parameterName=Spec Intense})
  )
  NamedRerouteDeclaration(None=LinearInterpolate(A=Add(A=@n14, B=@n21), B=VectorParameter() {parameterName=CrackColor, defaultValue=<rgba 1 1 1 1>}.RGB, Alpha=NamedRerouteUsage())) {name=EmissiveOUT, nodeColor=<rgba 1 0.4219 0 1>}
  n22 = MaterialFunctionCall(Float2=TextureCoordinate()) {materialFunction=&BreakOutFloat2Components}
  n23 = ScalarParameter() {defaultValue=1, group=Normal, parameterName=Normal Map Tile Rate On U}
  n24 = ScalarParameter() {defaultValue=0.2, group=Normal, parameterName=Normal Map Tile Rate On V}
  n25 = TextureSampleParameter2D(UVs=AppendVector(A=Multiply(A=@n22.R, B=@n23), B=Multiply(A=@n22.G, B=@n24))) {group=Normal, parameterName=Normal, samplerType=SAMPLERTYPE_Normal, texture=/Game/Global/Textures/Default/Noise/T_Noise_Normal_01}
  n26 = Normalize(VectorInput=Multiply(A=@n25.RGB, B=VectorParameter() {group=Normal, parameterName=Normal Intensity, defaultValue=<rgba 0.1 0.1 1 0>}.RGB))
  n27 = StaticBoolParameter() {defaultValue=true, parameterName=UseNormalTexture}
  n28 = StaticSwitch(True=Transform(None=@n26), False=VertexNormalWS(), Value=@n27)
  n29 = Multiply(
   A=ScalarParameter() {defaultValue=1, parameterName=Diffuse-FresnelAmount}
   B=Fresnel(ExponentIn=ScalarParameter() {parameterName=Diffuse-FresnelExponent}, Normal=@n28)
  )
  n30 = VectorParameter() {parameterName=Base Color}
  NamedRerouteDeclaration(None=LinearInterpolate(A=Add(A=@n29, B=@n30.RGB), B=VectorParameter() {parameterName=CrackColor, defaultValue=<rgba 1 1 1 1>}.RGB, Alpha=NamedRerouteUsage())) {name=BaseColor_OUT, nodeColor=<rgba 0 1 0.9766 1>}
  NamedRerouteDeclaration(None=LinearInterpolate(B=ComponentMask(None=@n8.Result) {r=true}, Alpha=@n10)) {name=CrackMask, nodeColor=<rgba 0.1953 1 0 1>}
  n31 = MaterialFunctionCall!(Texture=@n1, Position=@n5.PS, Normal=@n3.PS, TextureSize=@n2, Pan, BlendContrast=@n4) {materialFunction=&/MF_BiplanarNormal_Pan}
  n32 = ScalarParameter() {defaultValue=-10, parameterName=CrackNormalIntensity}
  NamedRerouteDeclaration(None=LinearInterpolate(
    A=Constant3Vector() {constant=<rgba 0 0 1 0>}
    B=MaterialFunctionCall(Normal=@n31.TangentSpaceNormal, Flatness=@n32) {materialFunction=&FlattenNormal}.Result
    Alpha=@n10
   )) {name=CrackNormal, nodeColor=<rgba 1 0.8438 0 1>}
  n33 = ScalarParameter() {parameterName=Opacity-Fresnel Inner Transparency}
  n34 = ScalarParameter() {defaultValue=0.85, parameterName=Opacity-Fresnel Outer Transparency}
  n35 = Saturate(None=Fresnel(ExponentIn=ScalarParameter() {defaultValue=4, parameterName=Opacity-Fresnel Exponent}, Normal=@n28))
  NamedRerouteDeclaration(None=LinearInterpolate(A=LinearInterpolate(A=@n33, B=@n34, Alpha=@n35), Alpha=NamedRerouteUsage())) {name=Opacity_OUT, nodeColor=<rgba 0 0.5078 1 1>}
  n36 = Constant3Vector() {constant=<rgba 0 0 1 0>}
  NamedRerouteDeclaration(None=MaterialFunctionCall(BaseNormal=StaticSwitch(True=@n26, False=@n36, Value=@n27), AdditionalNormal=NamedRerouteUsage()) {materialFunction=&BlendAngleCorrectedNormals}.Result) {name=Normal_OUT, nodeColor=<rgba 0.7344 0 1 1>}
  ScalarParameter() {defaultValue=1}
 ]
}
